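# Translucent lit Niagara ribbon material for the mist trail in the AccentWall emote.
# Color multiplies particle color by time-of-day tints and a dynamic-parameter fake candle light sphere mask, whose position uses only CandlePosX for all axes.
# Two panned smoke textures blend, then get density shaping, dynamic erosion and a U-axis mask.
# That alpha takes particle alpha and view-angle fade, adding depth fade only on High and Epic quality.
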Material /BRCosmetics/Animation/Game/MainPlayer/Emotes/AccentWall/FX/MI_Emote_AccentWall_MistTrail
graph {
 properties {
  bAutomaticallySetUsageInEditor=false
  bUsedWithNiagaraRibbons=true
  blendMode=BLEND_Translucent
  pixelDepthOffsetMode=PDOM_Legacy
  usedShadingModels=Default Lit
 }
 nodes [
  n1 = ParticleColor()
  n2 = MaterialFunctionCall(
   Color_Morning=VectorParameter() {parameterName=Color Morning Multiplier, defaultValue=<rgba 0.9896 0.7422 0.8193 0>}.RGB
   Color_Day=VectorParameter() {parameterName=Color Day Multiplier, defaultValue=<rgba 1 1 1 0>}.RGB
   Color_Evening=VectorParameter() {parameterName=Color Evening Multiplier, defaultValue=<rgba 0.599 0.2901 0.3358 0>}.RGB
   Color_Night=VectorParameter() {parameterName=Color Night Multiplier, defaultValue=<rgba 0.2583 0.3875 0.6458 0>}.RGB
  ) {materialFunction=&Color_Time_of_Day}
  n3 = DynamicParameter() {parameterIndex=1, defaultValue=<rgba 0 0 0 1>, paramNames [CandlePosX,CandlePosY,CandlePosZ,Param4]}
  n4 = LinearInterpolate(
   B=VectorParameter() {parameterName=FakeLightColor, defaultValue=<rgba 10 4 1 1>}.RGB
   Alpha=SphereMask(
    A=WorldPosition().XYZ
    B=MaterialFunctionCall(X=@n3.CandlePosX, Y=@n3.CandlePosX, Z=@n3.CandlePosX) {materialFunction=&MakeFloat3}.Result
    Radius=ScalarParameter() {defaultValue=30, parameterName=FakeLightRadius}
    Hardness=ScalarParameter() {defaultValue=-1, parameterName=FakeLightHardness}
   ) {attenuationRadius=0.5, hardnessPercent=0}
  ) {constA=1}
  n5 = DynamicParameter() {defaultValue=<rgba 1 0 0 1>, paramNames [BaseTextureOffset, Erosion, PanSpeed, FakeLightIntensity]}
  MaterialFunctionCall(Emissive=Multiply(A=Multiply(A=@n1.RGB, B=@n2.Result), B=LinearInterpolate(B=@n4, Alpha=@n5.FakeLightIntensity) {constA=1})) {materialFunction=&MF_FX_EmissiveAdjustments}
  n6 = TextureCoordinate()
  n7 = MaterialFunctionCall(X=ScalarParameter() {defaultValue=0.5, group=UV, parameterName=UV X Scale}, Y=Constant() {r=1}) {materialFunction=&MakeFloat2}
  n8 = DynamicParameter() {defaultValue=<rgba 1 0 0 1>, paramNames [BaseTextureOffset, Erosion, PanSpeed, FakeLightIntensity]}
  n9 = Constant()
  n10 = Panner(
   Coordinate=Add(
    A=Multiply(A=@n6, B=@n7.Result)
    B=MaterialFunctionCall(X=@n8.BaseTextureOffset, Y=@n9) {materialFunction=&MakeFloat2}.Result
   )
   Speed=MaterialFunctionCall(
    X=DynamicParameter() {defaultValue=<rgba 1 0 0 1>, paramNames [BaseTextureOffset, Erosion, PanSpeed, FakeLightIntensity]}.PanSpeed
    Y=Constant()
   ) {materialFunction=&MakeFloat2}.Result
  )
  n11 = MaterialFunctionCall(
   UVs=MaterialFunctionCall(XY=@n10) {materialFunction=&Swizzle}.YX
   Rotation Center=Constant() {r=0.5}
   "Rotation Angle (0-1)"=ScalarParameter() {defaultValue=0.5, parameterName=Param_1}
  ) {materialFunction=&CustomRotator}
  n12 = Multiply(
   A=@n10
   B=MaterialFunctionCall(X=ScalarParameter() {defaultValue=0.3, parameterName=Param_3}, Y=Constant() {r=1}) {materialFunction=&MakeFloat2}.Result
  )
  n13 = OneMinus(None=LinearInterpolate(
    A=TextureSample(UVs=@n11.Rotated Values) {texture=/Game/Tools/NewWaterfallGenerator/FX/Texture/T_SingleFrameFlowSmoke}.R
    B=TextureSample(UVs=@n12) {texture=/Game/Effects/Fort_Effects/Textures/Smoke/T_Puffy_Smoke_Trail_01}.R
    Alpha=ScalarParameter() {defaultValue=0.5, parameterName=Param_2}
   ))
  n14 = ScalarParameter() {defaultValue=2, parameterName=Density}
  n15 = Multiply(A=OneMinus(None=Power(Base=@n13, Exp=@n14)))
  n16 = MaterialFunctionCall!(
   Texture=@n15
   Erosion=OneMinus(None=DynamicParameter() {defaultValue=<rgba 1 0 0 1>, paramNames [BaseTextureOffset, Erosion, PanSpeed, FakeLightIntensity]}.Erosion)
   Softness=ScalarParameter() {defaultValue=1, parameterName=ErosionSoftness}
  ) {materialFunction=/Game/Effects/Niagara/MasterMaterials/Functions/MF_Erosion_Smoothstep_01}
  n17 = MaterialFunctionCall(Float2=TextureCoordinate()) {materialFunction=&BreakOutFloat2Components}
  n18 = Constant() {r=0.57}
  n19 = Constant() {r=0.6}
  NamedRerouteDeclaration(None=Multiply(
    A=Multiply(A=@n15, B=@n16.Result)
    B=MaterialFunctionCall(Texture=@n17.R, Erosion=@n18, Softness=@n19) {materialFunction=&MF_Erosion_Smoothstep_01}.Result
   )) {name=BaseTexture, nodeColor=<rgba 1 0 0.2109 1>}
  n20 = Multiply(
   A=Multiply(A=@n1.A, B=NamedRerouteUsage())
   B=MaterialFunctionCall("FadeSensitivity (optional)"=ScalarParameter() {defaultValue=2, group=Opacity, parameterName=VievAngleFade}) {materialFunction=&FadeBasedOnViewAngle}.Result
  )
  n21 = ScalarParameter() {defaultValue=2, parameterName=DepthFadeDistance}
  n22 = Multiply(A=@n20, B=Power(Base=DepthFade(FadeDistance=@n21), Exp=ScalarParameter() {defaultValue=6}))
  QualitySwitch(Default=@n20, Low=@n20, High=@n22, Medium=@n20, Epic=@n22)
 ]
}
