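# Niagara system for the dense steam puff of the AccentWall emote, fired from the owner along its Y axis.
# Burst sprites add smoke puffs, while a ribbon plume grows with a reveal value and bends toward a time-averaged owner rotation for trailing lag.
# User.Strength scales count, size, speed, length or opacity per emitter.
# Time-of-day and front-end modifiers adjust look, and skeletal-mesh sampled positions feed materials, including night-only candle-proximity values.
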NiagaraSystem /BRCosmetics/Animation/Game/MainPlayer/Emotes/AccentWall/FX/NS_Emote_AccentWall_Steam_Dense
dependencies {
 usedDataInterfaces [
  &NiagaraDataInterfaceCurve
  &NiagaraDataInterfaceColorCurve
  &NiagaraDataInterfaceArrayInt32
  &NiagaraDataInterfaceSkeletalMesh
 ]
 usedDynamicInputs [&SimulationPosition]
 usedModules [
  &self:SystemSpawnScript.SystemScriptSource.SystemScriptGraph.NiagaraNodeAssignment_3.SetVariables_#1
  &SystemState
  &self:SystemSpawnScript.SystemScriptSource.SystemScriptGraph.NiagaraNodeAssignment_4.SetVariables_#2
  &self:SystemSpawnScript.SystemScriptSource.SystemScriptGraph.NiagaraNodeAssignment_2.SetVariables_#3
  &InterpolateOverTime
  &self:SystemSpawnScript.SystemScriptSource.SystemScriptGraph.NiagaraNodeAssignment_5.SetVariables_#4
 ]
}
emitters {
 SmokeBurst {
  emitterUpdateScript [
   &EmitterState()
   &SpawnBurst_Instantaneous(
    Spawn Count=&Multiply_Int(
     A=&RandomRangeInt(Minimum=1, Maximum=2)
     B=&SelectIntFromArray(
      Direct Array Index=&MakeIntFromFloat(Float=$User.Strength)
      Int Selection Array=DataInterface() {propertyValues {IntData [1,2]}}
     )
    )
    Spawn Time=0.03
    Use Spawn Probability=false
    UseLoopCountLimit=false
   )
   /SetVariables(Emitter.ShadowStrength:NiagaraFloat=&FNTimeOfDayScalar(Night Scale=1, Morning Scale=0.7, Day Scale=0.7, Evening Scale=0.7)) {isSetParameters=true}
  ]
  particleSpawnScript [
   &InitializeParticle(
    Color Maximum=<rgba 0.4583 0.4583 0.4583 0.7>
    Color Minimum=<rgba 0.6823 0.6823 0.6823 0.8>
    Color Mode=::NewEnumerator2
    Lifetime=&UniformRangedFloat(Minimum=0.3, Maximum=0.5)
    Mass Mode=::NewEnumerator1
    Sprite Rotation Mode=::NewEnumerator2
    Sprite Size=&Vector2DFromFloat(Value=&UniformRangedFloat(Minimum=50, Maximum=50))
    Sprite Size Mode=::NewEnumerator4
    UsePositionOffset=false
    Write Material Random=true
   ) {module=InitializeParticle001}
   &ShapeLocation(
    MaskSpawnGroup=false
    Offset=&FN_FrontEndVector3Modifier(Gameplay=<0 -1 0>, Front End=<0 1 0>)
    Offset Coordinate Space=::Local
    Offset Mode=::NewEnumerator0
    Override Seed=false
    Sphere Radius=0.1
    UseOwnerScale=true
   ) {module=ShapeLocation001}
   &AddVelocity(
    Cone Angle=5
    Cone Axis=$Engine.Owner.SystemYAxis
    Module.UseConeFalloff=false
    Module.UseVelDistribution=false
    Override Seed=false
    Rotation Coordinate Space=::World
    Velocity Mode=::NewEnumerator2
    Velocity Speed=&Multiply_Float(
     A=&RandomRangeFloat(Minimum=20, Maximum=80)
     B=&Multiply_Float(
      A=&Lerp_Float(A=0.7, B=1, Alpha=$System.SizeRandom)
      B=&Lerp_Float(A=0.6, B=1, Alpha=$User.Strength)
     )
    )
   ) {module=AddVelocity001}
  ]
  particleUpdateScript [
   &ParticleState()
   &SpriteRotationRate(Rotation Rate=&Multiply_Float(
     A=&RandomRangeFloat(Minimum=100, Maximum=150)
     B=&UniformAOrBFloat(A=-1, Evaluation Type=::NewEnumerator0, Recalculate Random Each Loop=true, Override Seed=false)
    ))
   &SolveForcesAndVelocity(Clamp Velocity=false, Limit Acceleration=false)
   &ScaleSpriteSize(Scale Factor=&Vector2DFromFloat(Value=&FloatFromCurve(
      FloatCurve=DataInterface() {propertyValues {Curve {keys [<at 0 0.1>,<at 1 1>]}}}
      Scale Curve=&Multiply_Float(
       A=0.7
       B=&Multiply_Float(
        A=&Lerp_Float(A=0.7, B=1, Alpha=$System.SizeRandom)
        B=&Lerp_Float(A=0.8, B=1, Alpha=$User.Strength)
       )
      )
     ))) {module=ScaleSpriteSize001}
   &CameraOffset(Camera Offset Amount=&FN_FrontEndFloatModifier(Gameplay=3, Front End=0))
   &ScaleColor(
    Scale Alpha=&FloatFromCurve(
     FloatCurve=DataInterface() {propertyValues {Curve {keys [
         {interpMode=RCIM_Cubic, tangentMode=RCTM_User, time=0.1, value=0.6}
         {interpMode=RCIM_Cubic, tangentMode=RCTM_User, time=0.2, value=1}
         {tangentMode=RCTM_User, time=1, value=0}
        ]}}}
     Scale Curve=&Multiply_Float(A=0.666, B=&FNTimeOfDayScalar(Night Scale=2.2, Morning Scale=1, Day Scale=1.1, Evening Scale=1))
    )
    Scale RGB=&Multiply_Vector(
     A=&VectorFromFloat(Value=&FNTimeOfDayScalar(Night Scale=1, Morning Scale=1, Day Scale=0.8, Evening Scale=1))
     B=&MakeVectorFromLinearColorRGB(Color=&FNTimeOfDayColor(Night Color=<rgba 0.335 0.39 0.5 1>, Morning Color=<rgba 1 1 1 1>, Day Color=<rgba 1 1 1 1>, Evening Color=<rgba 1 1 1 1>))
    )
    ScaleA=true
    ScaleRGB=true
   )
   &SkeletalMeshLocation(
    MaskSpawnGroup=false
    Mesh Sampling Type=::NewEnumerator5
    Orientation Sampling=::NewEnumerator1
    Position Sampling=::NewEnumerator2
    Sampled Position Coordinate Space=::World
    Sampled Position Offset Coordinate Space:ENiagaraSkelMeshTransforms=::NewEnumerator1
    Skeletal Mesh=$System.Skeletal Mesh
    Socket Sampling Mode=::NewEnumerator5
   )
   /SetVariables(Particles.DistFromCandle:NiagaraFloat=&Subtract_Float(
     A=&Divide_Float(
      A=&DistanceBetweenPositions(
       A=&TransformPosition(Position=$Particles.Position, Source Space=::Local, Destination Space=::World)
       B=$Particles.SkeletalMeshLocation.SampledPosition
      )
      B=16
     )
     B=1
    )) {isSetParameters=true}
   &DynamicMaterialParameters(
    Index 0 Param 1=&Add_Float(
     A=&RandomRangeFloat(Minimum=0, Maximum=5)
     B=&FloatFromCurve(
      FloatCurve=DataInterface() {propertyValues {Curve {keys [
          {arriveTangent=2.7533, interpMode=RCIM_Cubic, leaveTangent=1.8958, tangentMode=RCTM_User, time=0, value=1}
          {arriveTangent=0.8292, interpMode=RCIM_Cubic, leaveTangent=0.8292, tangentMode=RCTM_User, time=0.321, value=1.3453}
          {arriveTangent=0.0529, interpMode=RCIM_Cubic, leaveTangent=0.2021, tangentMode=RCTM_User, time=1, value=1.5}
         ]}}}
      CurveIndex=&Add_Float(A=&Multiply_Float(A=$Particles.Initial.MaterialRandom, B=0.5), B=$Particles.NormalizedAge)
     )
    )
    Index 0 Param 2=&FloatFromCurve(FloatCurve=DataInterface() {propertyValues {Curve {keys [
         {arriveTangent=0.0453, interpMode=RCIM_Cubic, tangentMode=RCTM_Break, time=0, value=-1}
         {arriveTangent=0.0604, interpMode=RCIM_Cubic, leaveTangent=0.0604, tangentMode=RCTM_User, time=1, value=-1}
        ]}}})
    Index 0 Param 3=&FloatFromCurve(
     FloatCurve=DataInterface() {propertyValues {Curve {keys [{arriveTangent=-3.6642, interpMode=RCIM_Cubic, leaveTangent=-3.6642, tangentMode=RCTM_User, time=0, value=1}]}}}
     Scale Curve=&Multiply_Float(
      A=&Multiply_Float(A=0.002, B=&FNTimeOfDayScalar(Night Scale=1, Morning Scale=0, Day Scale=0, Evening Scale=0))
      B=&Lerp_Float(A=1, B=0, Alpha=&ClampFloat(Float=$Particles.DistFromCandle))
     )
    )
    Index 0 Param 4=&FloatFromCurve(FloatCurve=DataInterface() {propertyValues {Curve {keys [<at 0.3 0>,<at 0.4 1>,<at 1 0>]}}})
    Index 1 Param 1=&FloatFromCurve(
     FloatCurve=DataInterface() {propertyValues {Curve {keys [
         {arriveTangent=2.8574, interpMode=RCIM_Cubic, leaveTangent=2.8574, tangentMode=RCTM_User, time=0, value=1}
         {arriveTangent=4.2257, interpMode=RCIM_Cubic, leaveTangent=4.2257, tangentMode=RCTM_User, time=0.25, value=1}
        ]}}}
     Scale Curve=&RandomRangeFloat(Minimum=0.5, Maximum=1)
    )
    Index 1 Param 2=&FloatFromCurve(
     FloatCurve=DataInterface() {propertyValues {Curve {keys [<at 0 0.2>, {leaveTangent=0.9066, time=1, value=0.3}]}}}
     Scale Curve=&RandomRangeFloat(Minimum=0.5, Maximum=0.6)
    )
    Index 1 Param 3=&FloatFromCurve(
     FloatCurve=DataInterface() {propertyValues {Curve {keys [
         {arriveTangent=2.0526, interpMode=RCIM_Cubic, leaveTangent=2.0526, tangentMode=RCTM_User, time=0, value=0}
         {interpMode=RCIM_Cubic, tangentMode=RCTM_User, time=1, value=1}
        ]}}}
     Scale Curve=0
    )
    Index 1 Param 4=&FloatFromCurve(
     FloatCurve=DataInterface() {propertyValues {Curve {keys [{arriveTangent=-0.3224, interpMode=RCIM_Cubic, leaveTangent=-0.3224, tangentMode=RCTM_User, time=0, value=1}]}}}
     Scale Curve=0
    )
    Index 2 Param 1=&FloatFromCurve(FloatCurve=DataInterface() {propertyValues {Curve {keys [
         {interpMode=RCIM_Cubic, leaveTangent=4.2485, tangentMode=RCTM_Break, time=0, value=0}
         {arriveTangent=1.0548, interpMode=RCIM_Cubic, leaveTangent=1.0548, tangentMode=RCTM_User, time=0.17, value=0.5715}
         <at 1 1>
        ]}}})
    Index 2 Param 2=&FloatFromCurve(
     FloatCurve=DataInterface() {propertyValues {Curve {keys [
         {arriveTangent=1.9056, interpMode=RCIM_Cubic, leaveTangent=1.9056, tangentMode=RCTM_User, time=0, value=0.2}
         {arriveTangent=0.0259, interpMode=RCIM_Cubic, leaveTangent=0.0259, tangentMode=RCTM_User, time=1, value=1.4}
        ]}}}
     Scale Curve=&RandomRangeFloat(Minimum=0.5, Maximum=0.7)
    )
    Index 2 Param 3=&FloatFromCurve(FloatCurve=DataInterface() {propertyValues {Curve {keys [
         {interpMode=RCIM_Cubic, leaveTangent=2.3064, tangentMode=RCTM_Break, time=0, value=0}
         {arriveTangent=2, interpMode=RCIM_Cubic, leaveTangent=2, time=0.2, value=1.4}
         {interpMode=RCIM_Cubic, leaveTangent=11.1233, tangentMode=RCTM_Break, time=1, value=2}
        ]}}})
    Index 2 Param 4=&FloatFromCurve(
     FloatCurve=DataInterface() {propertyValues {Curve {keys [
         {arriveTangent=-0.6867, interpMode=RCIM_Cubic, leaveTangent=-0.6867, tangentMode=RCTM_User, time=0, value=1}
         {interpMode=RCIM_Cubic, leaveTangent=-1.2956, tangentMode=RCTM_Break, time=1, value=0.75}
        ]}}}
     Scale Curve=1.333
    )
    Module.Param0WriteEnabled=true
    Module.Param0WriteEnabled1=true
    Module.Param0WriteEnabled2=true
    Module.Param1WriteEnabled=false
    Module.Param1WriteEnabled1=true
    Module.Param1WriteEnabled2=true
    Module.Param2WriteEnabled=true
    Module.Param2WriteEnabled1=true
    Module.Param2WriteEnabled2=true
    Module.Param3WriteEnabled=true
    Module.Param3WriteEnabled1=true
    Module.Param3WriteEnabled2=true
    Write Parameter Index 1=true
    Write Parameter Index 2=true
   ) {module=DynamicMaterialParameters001}
  ]
  rendererClasses [&NiagaraSpriteRendererProperties]
  renderers [{rendererClass=&NiagaraSpriteRendererProperties}]
 }
 SmokeRibbon {
  emitterSpawnScript [/SetVariables(
    Emitter.Length:NiagaraFloat=&Lerp_Float(A=80, B=122, Alpha=$User.Strength)
    Emitter.ParticlesToSpawn=12
    Emitter.Random:NiagaraFloat=&RandomRangeFloat(Minimum=0, Maximum=1)
   ) {isSetParameters=true}]
  emitterUpdateScript [
   &EmitterState(Inactive Response=::NewEnumerator1, Life Cycle Mode=::NewEnumerator1, Loop Behavior=::NewEnumerator1, Loop Duration=1, UseLoopDelay=false)
   &SpawnBurst_Instantaneous(Spawn Count=$Emitter.ParticlesToSpawn, Use Spawn Probability=false, UseLoopCountLimit=false)
   /SetVariables(Emitter.Reveal:NiagaraFloat=&Multiply_Float(
     A=&Lerp_Float(
      A=&FloatFromCurve(FloatCurve=DataInterface() {propertyValues {Curve {keys [<at 0 0>,<at 0.6 1>]}}}, CurveIndex=$System.MainReveal)
      B=&FloatFromCurve(FloatCurve=DataInterface() {propertyValues {Curve {keys [<at 0 0>,<at 0.95 1>]}}}, CurveIndex=$System.MainReveal)
      Alpha=$System.SystemRandom
     )
     B=&Lerp_Float(A=1, B=1.1, Alpha=$Emitter.Random)
    )) {isSetParameters=true}
   /SetVariables(Emitter.Length:NiagaraFloat=&Multiply_Float(
     A=$Emitter.Initial.Length
     B=&FloatFromCurve(
      FloatCurve=DataInterface() {propertyValues {Curve {keys [
          {interpMode=RCIM_Cubic, leaveTangent=6.4155, tangentMode=RCTM_Break, time=0, value=0}
          {arriveTangent=0.1191, interpMode=RCIM_Cubic, leaveTangent=0.1191, tangentMode=RCTM_User, time=0.1, value=0.9}
          {arriveTangent=0.1177, interpMode=RCIM_Cubic, leaveTangent=0.1177, tangentMode=RCTM_User, time=1, value=1}
         ]}}}
      CurveIndex=$Emitter.Reveal
      Scale Curve=&Lerp_Float(A=0.7, B=0.8, Alpha=$System.SizeRandom)
     )
    )) {isSetParameters=true}
  ]
  particleSpawnScript [&InitializeParticle(Color Mode=::NewEnumerator0, Sprite Rotation Mode=::NewEnumerator2, UsePositionOffset=false, Write Material Random=true)]
  particleUpdateScript [
   &ParticleState(Kill Particles When Lifetime Has Elapsed=false)
   /SetVariables(Particles.ParticleRelativeLifetimeRibbons:NiagaraFloat=&ReturnNormalizedExecIndex()) {isSetParameters=true}
   /SetVariables(Particles.OpacityMultiplierRibbons:NiagaraFloat=&FloatFromCurve(
     FloatCurve=DataInterface() {propertyValues {Curve {keys [
         {interpMode=RCIM_Cubic, tangentMode=RCTM_User, time=0, value=0}
         {arriveTangent=0.007, interpMode=RCIM_Cubic, leaveTangent=0.007, tangentMode=RCTM_User, time=0.5, value=1.5}
         {arriveTangent=-0.7513, interpMode=RCIM_Cubic, leaveTangent=-0.7513, tangentMode=RCTM_User, time=0.8, value=1}
         {arriveTangent=-0.0252, interpMode=RCIM_Cubic, leaveTangent=-0.0252, tangentMode=RCTM_User, time=1.7, value=0}
        ]}}}
     CurveIndex=&Add_Float(
      A=&Multiply_Float(A=$Particles.ParticleRelativeLifetimeRibbons, B=-1)
      B=&Multiply_Float(A=$Emitter.Reveal, B=2)
     )
    )) {isSetParameters=true}
   /SetVariables(Particles.InterpolateCurveRibbons:NiagaraFloat=&FloatFromCurve(
     FloatCurve=DataInterface() {propertyValues {Curve {keys [{arriveTangent=-0.6503, time=0, value=0}, {arriveTangent=0.8679, time=1, value=1}]}}}
     CurveIndex=$Particles.ParticleRelativeLifetimeRibbons
    )) {isSetParameters=true}
   /SetVariables(Particles.PositionRibbons:NiagaraPosition=&AddVectorToPosition(
     Position=&AddVectorToPosition(
      Position=/Niagara/DynamicInputs/Helpers/SimulationPosition()
      Vector=&TransformVector(Vector=<0 -1.5 0>, Destination Space=::Simulation)
     )
     Vector=&Multiply_VectorByFloat(
      Vector=&RotateVectorByQuaternion(
       Vector To Rotate=<0 1 0>
       Quaternion=&LerpQuaternion(
        Quaternion A:Quat4f=$Engine.Owner.Rotation
        Quaternion B:Quat4f=$System.InterpolateOverTime.InterpolateOverTime_Quaternion.Moving Average
        Lerp Factor:NiagaraFloat=$Particles.InterpolateCurveRibbons
        Enable Spherical Linear Interpolation:NiagaraBool=false
       )
      )
      Float=&Multiply_Float(A=$Emitter.Length, B=$Particles.InterpolateCurveRibbons)
     )
    )) {isSetParameters=true}
   /SetVariables(
    Particles.RibbonColor:LinearColor=&MultiplyLinearColor(
     Linear Color=&MultiplyLinearColor(
      Linear Color=&ColorFromCurve(
       ColorCurve=DataInterface() {propertyValues {
         AlphaCurve {keys [
           <at 0 0.8>
           <at 0.0346 1>
           <at 0.2133 1>
           {arriveTangent=-1.2711, interpMode=RCIM_Cubic, leaveTangent=-1.2711, time=0.6348, value=0.8}
           <at 1 0 RCIM_Cubic>
          ]}
         BlueCurve {keys [<at 0.4438 0.666>]}
         GreenCurve {keys [<at 0.4438 0.666>]}
         RedCurve {keys [<at 0.4438 0.666>]}
        }}
       CurveIndex=$Particles.ParticleRelativeLifetimeRibbons
      )
      Scale Factor=&MakeVector4_FromLinearColor(LinearColor=&FNTimeOfDayColor(Night Color=<rgba 0.335 0.39 0.5 1>, Morning Color=<rgba 1 1 1 1>, Day Color=<rgba 1 1 1 1>, Evening Color=<rgba 1 1 1 1>))
     )
     Scale Factor=&MakeFloat4FromVectorAndFloat(
      "Vector (RGB)"=&VectorFromFloat(Value=1)
      "Float (Alpha)"=&Multiply_Float(A=&FN_FrontEndFloatModifier(Gameplay=1, Front End=0.8), B=1)
     )
    )
    Particles.RibbonWidth=&FloatFromCurve(
     FloatCurve=DataInterface() {propertyValues {Curve {keys [
         {arriveTangent=1.1559, interpMode=RCIM_Cubic, leaveTangent=1.1559, tangentMode=RCTM_User, time=0, value=0.15}
         {arriveTangent=1.0523, interpMode=RCIM_Cubic, leaveTangent=1.0523, tangentMode=RCTM_User, time=0.2799, value=0.4747}
         {arriveTangent=-0.0372, interpMode=RCIM_Cubic, leaveTangent=-0.0372, tangentMode=RCTM_User, time=0.7, value=0.8}
         {arriveTangent=-1.5957, interpMode=RCIM_Cubic, leaveTangent=-1.5957, tangentMode=RCTM_User, time=1, value=0.5138}
        ]}}}
     CurveIndex=$Particles.InterpolateCurveRibbons
     Scale Curve=33
    )
   ) {isSetParameters=true}
   /SetVariables(Particles.ScaledRibbonWidth:NiagaraFloat=&Multiply_Float(
     A=$Particles.RibbonWidth
     B=&Multiply_Float(
      A=&FloatFromCurve(
       FloatCurve=DataInterface() {propertyValues {Curve {keys [
           {interpMode=RCIM_Cubic, leaveTangent=0.5491, tangentMode=RCTM_Break, time=0, value=0.7}
           {arriveTangent=1.0086, interpMode=RCIM_Cubic, leaveTangent=1.0086, tangentMode=RCTM_User, time=0.3931, value=1}
           {arriveTangent=0.7172, interpMode=RCIM_Cubic, tangentMode=RCTM_Break, time=1, value=1.5}
          ]}}}
       CurveIndex=$Emitter.Reveal
       Scale Curve=&Lerp_Float(
        A=1.6
        B=&FloatFromCurve(
         FloatCurve=DataInterface() {propertyValues {Curve {keys [<at 0 1.8 RCIM_Constant>,<at 0.333 1.9 RCIM_Constant>,<at 0.666 2 RCIM_Constant>]}}}
         CurveIndex=$Emitter.Random
         Scale Curve=0.9
        )
        Alpha=$System.SizeRandom
       )
      )
      B=&Lerp_Float(A=0.7, B=1, Alpha=$User.Strength)
     )
    )) {isSetParameters=true}
   /SetVariables(Particles.ScaledRibbonColor:LinearColor=&MultiplyLinearColor(
     Linear Color=$Particles.RibbonColor
     Scale Factor=&MakeFloat4FromVectorAndFloat(
      "Vector (RGB)"=<1 1 1>
      "Float (Alpha)"=&Multiply_Float(
       A=&Multiply_Float(
        A=$Particles.OpacityMultiplierRibbons
        B=&FNTimeOfDayScalar(Night Scale=2, Morning Scale=0.9, Day Scale=0.8, Evening Scale=0.7)
       )
       B=&Multiply_Float(
        A=&FloatFromCurve(
         FloatCurve=DataInterface() {propertyValues {Curve {keys [
             {interpMode=RCIM_Cubic, tangentMode=RCTM_Break, time=0.6, value=1.3}
             {arriveTangent=-4.2498, interpMode=RCIM_Cubic, tangentMode=RCTM_Break, time=0.99, value=0}
            ]}}}
         CurveIndex=$Emitter.Reveal
         Scale Curve=&Lerp_Float(A=0.8, B=1, Alpha=$Emitter.Random)
        )
        B=&Multiply_Float(
         A=&Multiply_Float(A=$System.OpacityRandom, B=&FN_FrontEndFloatModifier(Gameplay=1, Front End=0.5))
         B=&Lerp_Float(A=0.4, B=1, Alpha=$User.Strength)
        )
       )
      )
     )
    )) {isSetParameters=true}
   &CameraOffset(Camera Offset Amount=5)
   &SkeletalMeshLocation(
    MaskSpawnGroup=false
    Mesh Sampling Type=::NewEnumerator5
    Orientation Sampling=::NewEnumerator1
    Position Sampling=::NewEnumerator2
    Sampled Position Coordinate Space=::World
    Sampled Position Offset Coordinate Space:ENiagaraSkelMeshTransforms=::NewEnumerator1
    Skeletal Mesh=$System.Skeletal Mesh
    Socket Sampling Mode=::NewEnumerator5
   )
   &DynamicMaterialParameters(
    Index 0 Param 1=$Emitter.Random
    Index 0 Param 2=&Multiply_Float(
     A=&MakeFloatFromVector4(Vector 4=&MakeVector4_FromLinearColor(LinearColor=$Particles.RibbonColor), Channel=::NewEnumerator3)
     B=$Particles.OpacityMultiplierRibbons
    )
    Index 0 Param 3=0.4
    Index 0 Param 4=&FNTimeOfDayScalar(Night Scale=2, Morning Scale=0, Day Scale=0, Evening Scale=0)
    Index 1 Param 1=&MakeFloatFromVector(Vector=$Particles.SkeletalMeshLocation.SampledPosition, Channel=::NewEnumerator0)
    Index 1 Param 2=&MakeFloatFromVector(Vector=$Particles.SkeletalMeshLocation.SampledPosition, Channel=::NewEnumerator1)
    Index 1 Param 3=&MakeFloatFromVector(Vector=$Particles.SkeletalMeshLocation.SampledPosition, Channel=::NewEnumerator2)
    Module.Index0Param0Write=true
    Module.Index0Param1Write=true
    Module.Index0Param2Write=true
    Module.Index0Param3Write=true
    Module.Index1Param0Write=true
    Module.Index1Param1Write=true
    Module.Index1Param2Write=true
    Module.Index1Param3Write=true
    Write Parameter Index 1=true
   )
  ]
  rendererClasses [&NiagaraRibbonRendererProperties]
  renderers [{rendererClass=&NiagaraRibbonRendererProperties}]
 }
 SmokeSprites {
  emitterUpdateScript [
   &EmitterState()
   &SpawnBurst_Instantaneous(
    Spawn Count=&Multiply_Int(
     A=&RandomRangeInt(Minimum=2, Maximum=2)
     B=&SelectIntFromArray(
      Direct Array Index=&MakeIntFromFloat(Float=$User.Strength)
      Int Selection Array=DataInterface() {propertyValues {IntData [2,4]}}
     )
    )
    Use Spawn Probability=false
    UseLoopCountLimit=false
   )
   /SetVariables(Emitter.ShadowStrength:NiagaraFloat=&FNTimeOfDayScalar(Night Scale=1, Morning Scale=0.7, Day Scale=0.7, Evening Scale=0.7)) {isSetParameters=true}
  ]
  particleSpawnScript [
   &InitializeParticle(
    Color Maximum=<rgba 0.4583 0.4583 0.4583 0.7>
    Color Minimum=<rgba 0.6823 0.6823 0.6823 0.8>
    Color Mode=::NewEnumerator2
    Lifetime Max=0.96
    Lifetime Min=0.7
    Lifetime Mode=::NewEnumerator1
    Position Offset Coordinate Space=::Local
    Sprite Rotation Mode=::NewEnumerator2
    Sprite Size=&Multiply_Vector2D_ByFloat(
     Float=&RandomRangeFloat(Minimum=63, Maximum=80)
     Vector2D=&Vector2DFromFloat(Value=&Lerp_Float(
       A=0.3
       B=1
       Alpha=&FloatFromCurve(
        FloatCurve=DataInterface() {propertyValues {Curve {keys [<at 0 0>,<at 1 1>]}}}
        CurveIndex=&ReturnNormalizedExecIndex()
       )
      ))
    )
    Sprite Size Mode=::NewEnumerator4
    UsePositionOffset=true
    Write Material Random=true
   ) {module=InitializeParticle001}
   &SystemLocation(MaskSpawnGroup=false, Offset=&FN_FrontEndVector3Modifier(Gameplay=<0 -1 0>, Front End=<0 1 0>))
   &AddVelocity(
    Cone Angle=20
    Cone Axis=$Engine.Owner.SystemYAxis
    Distribution Along Cone Axis=1
    Module.UseConeFalloff=false
    Module.UseVelDistribution=false
    Override Seed=false
    Rotation Coordinate Space=::World
    Velocity Mode=::NewEnumerator2
    Velocity Speed=&Multiply_Float(
     A=&Multiply_Float(A=200, B=&Lerp_Float(A=0.6, B=1, Alpha=$User.Strength))
     B=&Lerp_Float(A=0.1, B=1, Alpha=&ReturnNormalizedExecIndex())
    )
   ) {module=AddVelocity001}
  ]
  particleUpdateScript [
   &ParticleState()
   &SpriteRotationRate(Rotation Rate=&Multiply_Float(
     A=&UniformAOrBFloat(A=-1, Evaluation Type=::NewEnumerator0, Recalculate Random Each Loop=true, Override Seed=false)
     B=&FloatFromCurve(
      FloatCurve=DataInterface() {propertyValues {Curve {keys [
          {arriveTangent=-2.4575, interpMode=RCIM_Cubic, leaveTangent=-2.4575, tangentMode=RCTM_User, time=0, value=1}
          {arriveTangent=-0.2001, interpMode=RCIM_Cubic, leaveTangent=-0.2001, tangentMode=RCTM_User, time=0.4, value=0.1}
          {arriveTangent=-0.0902, interpMode=RCIM_Cubic, leaveTangent=-0.0902, tangentMode=RCTM_User, time=1, value=0.025}
         ]}}}
      Scale Curve=1600
     )
    ))
   &ScaleColor(
    Scale Alpha=&FloatFromCurve(
     FloatCurve=DataInterface() {propertyValues {Curve {keys [
         {interpMode=RCIM_Cubic, tangentMode=RCTM_User, time=0.1, value=0}
         {interpMode=RCIM_Cubic, tangentMode=RCTM_User, time=0.2, value=1}
         {tangentMode=RCTM_User, time=1, value=0}
        ]}}}
     Scale Curve=&Multiply_Float(
      A=&FNTimeOfDayScalar(Night Scale=2.2, Morning Scale=1, Day Scale=1.1, Evening Scale=1)
      B=&Multiply_Float(
       A=&Lerp_Float(A=0.3, B=1, Alpha=$User.Strength)
       B=&FN_FrontEndFloatModifier(Gameplay=1, Front End=0.3)
      )
     )
    )
    Scale RGB=&Multiply_Vector(
     A=&VectorFromFloat(Value=&FNTimeOfDayScalar(Night Scale=1, Morning Scale=1, Day Scale=0.8, Evening Scale=1))
     B=&MakeVectorFromLinearColorRGB(Color=&FNTimeOfDayColor(Night Color=<rgba 0.335 0.39 0.5 1>, Morning Color=<rgba 1 1 1 1>, Day Color=<rgba 1 1 1 1>, Evening Color=<rgba 1 1 1 1>))
    )
    ScaleA=true
    ScaleRGB=true
   )
   &Drag(
    Drag=&FloatFromCurve(
     FloatCurve=DataInterface() {propertyValues {Curve {keys [{interpMode=RCIM_Cubic, tangentMode=RCTM_Break, time=0, value=0}, {interpMode=RCIM_Cubic, leaveTangent=-2.586, time=0.7, value=1}]}}}
     Scale Curve=&Lerp_Float(A=5, B=9, Alpha=$Particles.Initial.MaterialRandom)
    )
    Use Linear Drag=true
    Use Rotational Drag=false
   )
   &AccelerationForce(Acceleration=&RandomRangeVector(Minimum=<0 0 50>, Maximum=<0 0 150>))
   &SolveForcesAndVelocity(Clamp Velocity=false, Limit Acceleration=false)
   &ScaleSpriteSize(Scale Factor=&Vector2DFromFloat(Value=&Multiply_Float(
      A=1
      B=&FloatFromCurve(
       FloatCurve=DataInterface() {propertyValues {Curve {keys [
           {arriveTangent=0.2741, interpMode=RCIM_Cubic, leaveTangent=0.2741, tangentMode=RCTM_User, time=0, value=0.2}
           {interpMode=RCIM_Cubic, tangentMode=RCTM_User, time=0.95, value=1}
           <at 1 0>
          ]}}}
       Scale Curve=&Lerp_Float(A=0.8, B=1, Alpha=$User.Strength)
      )
     )))
   &CameraOffset(Camera Offset Amount=5)
   &SkeletalMeshLocation(
    MaskSpawnGroup=false
    Mesh Sampling Type=::NewEnumerator5
    Orientation Sampling=::NewEnumerator1
    Position Sampling=::NewEnumerator2
    Sampled Position Coordinate Space=::World
    Sampled Position Offset Coordinate Space:ENiagaraSkelMeshTransforms=::NewEnumerator1
    Skeletal Mesh=$System.Skeletal Mesh
    Socket Sampling Mode=::NewEnumerator5
   )
   /SetVariables(Particles.DistFromCandle:NiagaraFloat=&Subtract_Float(
     A=&Divide_Float(
      A=&DistanceBetweenPositions(
       A=&TransformPosition(Position=$Particles.Position, Source Space=::Local, Destination Space=::World)
       B=$Particles.SkeletalMeshLocation.SampledPosition
      )
      B=16
     )
     B=1
    )) {isSetParameters=true}
   &DynamicMaterialParameters(
    Index 0 Param 1=&Add_Float(
     A=&RandomRangeFloat(Minimum=0, Maximum=5)
     B=&FloatFromCurve(
      FloatCurve=DataInterface() {propertyValues {Curve {keys [
          {arriveTangent=2.7533, interpMode=RCIM_Cubic, leaveTangent=1.8958, tangentMode=RCTM_User, time=0, value=1}
          {arriveTangent=0.8292, interpMode=RCIM_Cubic, leaveTangent=0.8292, tangentMode=RCTM_User, time=0.321, value=1.3453}
          {arriveTangent=0.0529, interpMode=RCIM_Cubic, leaveTangent=0.2021, tangentMode=RCTM_User, time=1, value=1.5}
         ]}}}
      CurveIndex=&Add_Float(A=&Multiply_Float(A=$Particles.Initial.MaterialRandom, B=0.5), B=$Particles.NormalizedAge)
     )
    )
    Index 0 Param 2=&FloatFromCurve(FloatCurve=DataInterface() {propertyValues {Curve {keys [
         {arriveTangent=0.0453, interpMode=RCIM_Cubic, tangentMode=RCTM_Break, time=0, value=-1}
         {arriveTangent=0.0604, interpMode=RCIM_Cubic, leaveTangent=0.0604, tangentMode=RCTM_User, time=1, value=-1}
        ]}}})
    Index 0 Param 3=&FloatFromCurve(
     FloatCurve=DataInterface() {propertyValues {Curve {keys [{arriveTangent=-3.6642, interpMode=RCIM_Cubic, leaveTangent=-3.6642, tangentMode=RCTM_User, time=0, value=1}]}}}
     Scale Curve=&Multiply_Float(
      A=&Multiply_Float(A=0.002, B=&FNTimeOfDayScalar(Night Scale=1, Morning Scale=0, Day Scale=0, Evening Scale=0))
      B=&Lerp_Float(A=1, B=0, Alpha=&ClampFloat(Float=$Particles.DistFromCandle))
     )
    )
    Index 0 Param 4=&FloatFromCurve(FloatCurve=DataInterface() {propertyValues {Curve {keys [<at 0.3 0>,<at 0.4 1>,<at 1 0>]}}}, Scale Curve=0.9)
    Index 1 Param 1=&FloatFromCurve(
     FloatCurve=DataInterface() {propertyValues {Curve {keys [
         {arriveTangent=2.8574, interpMode=RCIM_Cubic, leaveTangent=2.8574, tangentMode=RCTM_User, time=0, value=1}
         {arriveTangent=4.2257, interpMode=RCIM_Cubic, leaveTangent=4.2257, tangentMode=RCTM_User, time=0.25, value=1}
        ]}}}
     Scale Curve=&RandomRangeFloat(Minimum=0.5, Maximum=1)
    )
    Index 1 Param 2=&FloatFromCurve(
     FloatCurve=DataInterface() {propertyValues {Curve {keys [<at 0 0.2>, {leaveTangent=0.9066, time=1, value=0.3}]}}}
     Scale Curve=&RandomRangeFloat(Minimum=0.5, Maximum=0.6)
    )
    Index 1 Param 3=&FloatFromCurve(
     FloatCurve=DataInterface() {propertyValues {Curve {keys [
         {arriveTangent=2.0526, interpMode=RCIM_Cubic, leaveTangent=2.0526, tangentMode=RCTM_User, time=0, value=0}
         {interpMode=RCIM_Cubic, tangentMode=RCTM_User, time=1, value=1}
        ]}}}
     Scale Curve=0
    )
    Index 1 Param 4=&FloatFromCurve(
     FloatCurve=DataInterface() {propertyValues {Curve {keys [{arriveTangent=-0.3224, interpMode=RCIM_Cubic, leaveTangent=-0.3224, tangentMode=RCTM_User, time=0, value=1}]}}}
     Scale Curve=0
    )
    Index 2 Param 1=&FloatFromCurve(FloatCurve=DataInterface() {propertyValues {Curve {keys [
         {interpMode=RCIM_Cubic, leaveTangent=4.2485, tangentMode=RCTM_Break, time=0, value=0}
         {arriveTangent=1.0548, interpMode=RCIM_Cubic, leaveTangent=1.0548, tangentMode=RCTM_User, time=0.17, value=0.5715}
         <at 1 1>
        ]}}})
    Index 2 Param 2=&FloatFromCurve(
     FloatCurve=DataInterface() {propertyValues {Curve {keys [
         {arriveTangent=1.9056, interpMode=RCIM_Cubic, leaveTangent=1.9056, tangentMode=RCTM_User, time=0, value=0.2}
         {arriveTangent=0.0259, interpMode=RCIM_Cubic, leaveTangent=0.0259, tangentMode=RCTM_User, time=1, value=1.4}
        ]}}}
     Scale Curve=&RandomRangeFloat(Minimum=0.5, Maximum=0.7)
    )
    Index 2 Param 3=&FloatFromCurve(FloatCurve=DataInterface() {propertyValues {Curve {keys [
         {interpMode=RCIM_Cubic, leaveTangent=2.3064, tangentMode=RCTM_Break, time=0, value=0}
         {arriveTangent=2, interpMode=RCIM_Cubic, leaveTangent=2, time=0.2, value=1.4}
         {interpMode=RCIM_Cubic, leaveTangent=11.1233, tangentMode=RCTM_Break, time=1, value=2}
        ]}}})
    Index 2 Param 4=&FloatFromCurve(
     FloatCurve=DataInterface() {propertyValues {Curve {keys [
         {arriveTangent=-0.6867, interpMode=RCIM_Cubic, leaveTangent=-0.6867, tangentMode=RCTM_User, time=0, value=1}
         {interpMode=RCIM_Cubic, leaveTangent=-1.2956, tangentMode=RCTM_Break, time=1, value=0.75}
        ]}}}
     Scale Curve=1.333
    )
    Module.Param0WriteEnabled=true
    Module.Param0WriteEnabled1=true
    Module.Param0WriteEnabled2=true
    Module.Param1WriteEnabled=false
    Module.Param1WriteEnabled1=true
    Module.Param1WriteEnabled2=true
    Module.Param2WriteEnabled=true
    Module.Param2WriteEnabled1=true
    Module.Param2WriteEnabled2=true
    Module.Param3WriteEnabled=true
    Module.Param3WriteEnabled1=true
    Module.Param3WriteEnabled2=true
    Write Parameter Index 1=true
    Write Parameter Index 2=true
   )
   &CameraOffset(Camera Offset Amount=7) {module=CameraOffset001}
  ]
  rendererClasses [&NiagaraSpriteRendererProperties]
  renderers [{rendererClass=&NiagaraSpriteRendererProperties}]
 }
}
summary {
 systemName=NS_Emote_AccentWall_Steam_Dense
 emitters [
  {emitterName=SmokeRibbon, rendererClasses [&NiagaraRibbonRendererProperties]}
  {emitterName=SmokeSprites, rendererClasses [&NiagaraSpriteRendererProperties]}
  {emitterName=SmokeBurst, rendererClasses [&NiagaraSpriteRendererProperties]}
 ]
 userVariables [
  User.bIsFrontEnd {type=NiagaraBool}
  User.bIsFrontEndPreview {type=NiagaraBool}
  User.bIsLargeBodyPawn {type=NiagaraBool}
  User.CandleLightColor {type=LinearColor, description="", defaultValue {struct=LinearColor, value=<rgba 1 0.325 0.1 1>}}
  User.Reveal {type=NiagaraFloat, defaultValue=0.5, description=""}
  User.SpawnScale {type=NiagaraFloat, defaultValue=1, description=""}
  User.Strength {type=NiagaraFloat, defaultValue=1, description=""}
 ]
}
